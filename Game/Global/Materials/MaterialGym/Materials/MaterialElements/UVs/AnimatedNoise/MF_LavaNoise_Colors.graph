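# Material function that shades lava-lamp-style bubbles from Height and Fresnel inputs, outputting bubble colors and a warp vector.
# Height becomes a smoothed world normal, whose tangent-space XY serves as the warp output and samples a matcap.
# The world normal's Z blends bubble colors, which mix with scattering colors, weighted toward the lower scattering color as UV V rises.
# Matcap plus Fresnel, tinted by a vertical reflection gradient, is added on top.
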MaterialFunction /Game/Global/Materials/MaterialGym/Materials/MaterialElements/UVs/AnimatedNoise/MF_LavaNoise_Colors
graph {nodes [
  n1 = VectorParameter() {group=1. Colors, parameterName=Reflection Color Up, sortPriority=6, defaultValue=<rgba 0.61 0.61 0.019 1>}
  n2 = VectorParameter() {group=1. Colors, parameterName=Reflection Color Down, sortPriority=7, defaultValue=<rgba 0.92 0.92 0.027 1>}
  n3 = TextureCoordinate()
  n4 = LinearInterpolate(A=@n1.RGB, B=@n2.RGB, Alpha=Power(Base=ComponentMask(None=@n3) {g=true}) {constExponent=3})
  n5 = MaterialFunctionCall!(Height=FunctionInput() {inputName=Height, inputType=FunctionInput_Scalar}, Absolute World Position, World Space Vertex Normals) {materialFunction=/Engine/Functions/Engine_MaterialFunctions03/Procedurals/HeightToNormalSmooth}
  n6 = ComponentMask(None=Transform(None=@n5.WorldSpaceNormal) {transformSourceType=TRANSFORMSOURCE_World, transformType=TRANSFORM_Tangent}) {g=true, r=true}
  n7 = TextureSampleParameter2D(UVs=Add(A=Multiply(A=@n6) {constB=0.5}) {constB=0.5}) {group=5. Textures, parameterName=Matcap, sortPriority=2, texture=/Game/Environments/Asteria/Sets/Lair/Materials/LavaLamp/T_LavaLamp_Matcap}
  n8 = Add(
   A=Multiply(A=@n7.R, B=ScalarParameter() {defaultValue=1.4, group=1. Colors, parameterName=Matcap Intensity, sortPriority=9})
   B=Multiply(
    A=FunctionInput() {inputName=Fresnel, inputType=FunctionInput_Scalar}
    B=ScalarParameter() {defaultValue=1.4, group=1. Colors, parameterName=Fresnel Intensity, sortPriority=8}
   )
  )
  n9 = VectorParameter() {group=1. Colors, parameterName=Scattering Color Down, sortPriority=5, defaultValue=<rgba 0.61 0.61 0.0142 0.5>}
  n10 = VectorParameter() {group=1. Colors, parameterName=Scattering Color Up, sortPriority=4, defaultValue=<rgba 0.61 0.61 0.01 1>}
  n11 = VectorParameter() {group=1. Colors, parameterName=Bubbles Color Down, sortPriority=3, defaultValue=<rgba 0.05 0.05 0 1>}
  n12 = VectorParameter() {group=1. Colors, parameterName=Bubbles Color Up, sortPriority=2, defaultValue=<rgba 0.9322 0.84 0 1>}
  n13 = ComponentMask(None=@n5.WorldSpaceNormal) {b=true}
  n14 = ScalarParameter() {defaultValue=0.6, group=1. Colors, parameterName=Bubbles Color Shift, sortPriority=8}
  n15 = LinearInterpolate(A=@n10.RGB, B=LinearInterpolate(A=@n11.RGB, B=@n12.RGB, Alpha=Add(A=@n13, B=@n14) {constB=0.5}))
  n16 = ComponentMask(None=TextureCoordinate()) {g=true}
  n17 = Multiply(A=@n16, B=@n16)
  n18 = OneMinus(None=Multiply(A=@n9.A, B=Multiply(A=@n17, B=@n17)))
  FunctionOutput(None=Add(A=Multiply(A=@n4, B=@n8), B=LinearInterpolate(A=@n9.RGB, B=@n15, Alpha=@n18))) {outputName=Bubble Colors, sortPriority=1}
  FunctionOutput(None=@n6) {outputName=Bubble Warp, sortPriority=2}
 ]}
