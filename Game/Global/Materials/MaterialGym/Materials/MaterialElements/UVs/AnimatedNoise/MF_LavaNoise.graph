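# Material function that outputs animated, lava-lamp-style lava bubble colors from a UV input, under Seared Scurf parameters.
# Four scaled samples of one grayscale background texture pan at separate periodic speeds and multiply in pairs.
# Sharpened masks and inverted heights feed MF_LavaNoise_Colors.
# A static switch either colors all four layers as one product or colors each pair separately, blending by the second pair's inverted mask.
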MaterialFunction /Game/Global/Materials/MaterialGym/Materials/MaterialElements/UVs/AnimatedNoise/MF_LavaNoise
graph {nodes [
  n1 = AppendVector(
   A=ScalarParameter() {defaultValue=0.6, group=Seared Scurf, parameterName=Background Tiling X, sortPriority=0}
   B=ScalarParameter() {defaultValue=0.6, group=Seared Scurf, parameterName=Background Tiling Y, sortPriority=1}
  )
  n2 = ScalarParameter() {defaultValue=15, group=Seared Scurf, parameterName=Lava Global Tiling}
  n3 = Multiply(A=Add(A=NamedRerouteUsage()) {constB=-0.5}, B=Multiply(A=Multiply(A=@n1, B=@n2)) {constB=-1})
  n4 = Multiply(A=@n3) {constB=-1.4}
  n5 = Add(
   A=Multiply(A=ComponentMask(None=@n4) {r=true}) {constB=-1}
   B=Multiply(A=Time() {period=333.3333}, B=ScalarParameter() {defaultValue=-0.003, group=Seared Scurf, parameterName=Speed X - Pattern 01}) {constB=0.003}
  )
  n6 = Time() {period=100}
  n7 = ScalarParameter() {defaultValue=-0.5, group=Seared Scurf, parameterName=Speed Y - Pattern 01}
  n8 = Add(
   A=ComponentMask(None=@n4) {g=true}
   B=Multiply(A=Multiply(A=@n6, B=@n7) {constB=0.01}, B=ScalarParameter() {defaultValue=0.5, group=Seared Scurf, parameterName=Lava Overall Speed})
  )
  n9 = Multiply(A=@n3) {constB=-1.3}
  n10 = Add(
   A=ComponentMask(None=@n9) {r=true}
   B=Multiply(A=Time() {period=222.2222}, B=ScalarParameter() {defaultValue=-0.0045, group=Seared Scurf, parameterName=Speed X - Pattern 02}) {constB=-0.0045}
  )
  n11 = Time() {period=66.6667}
  n12 = ScalarParameter() {defaultValue=-0.5, group=Seared Scurf, parameterName=Speed Y - Pattern 02}
  n13 = Add(
   A=Multiply(A=ComponentMask(None=@n9) {g=true}) {constB=-1}
   B=Multiply(A=Multiply(A=@n11, B=@n12) {constB=-0.015}, B=ScalarParameter() {defaultValue=0.5, group=Seared Scurf, parameterName=Lava Overall Speed})
  )
  n14 = Multiply(
   A=TextureSample(UVs=AppendVector(A=@n5, B=@n8)) {samplerType=SAMPLERTYPE_LinearGrayscale, texture=/Game/Environments/Asteria/Sets/Lair/Materials/LavaLamp/T_LavaLamp_Background}.R
   B=TextureSample(UVs=AppendVector(A=@n10, B=@n13)) {samplerType=SAMPLERTYPE_LinearGrayscale, texture=&T_LavaLamp_Background}.R
  )
  n15 = Multiply(A=@n3) {constB=1.2}
  n16 = Add(
   A=ComponentMask(None=@n15) {r=true}
   B=Multiply(A=Time() {period=166.6667}, B=ScalarParameter() {defaultValue=-0.006, group=Seared Scurf, parameterName=Speed X - Pattern 03}) {constB=0.006}
  )
  n17 = Time() {period=50}
  n18 = ScalarParameter() {defaultValue=-0.1, group=Seared Scurf, parameterName=Speed Y - Pattern 03}
  n19 = Add(
   A=ComponentMask(None=@n15) {g=true}
   B=Multiply(A=Multiply(A=@n17, B=@n18) {constB=0.02}, B=ScalarParameter() {defaultValue=0.5, group=Seared Scurf, parameterName=Lava Overall Speed})
  )
  n20 = Multiply(A=@n3) {constB=1.1}
  n21 = Time() {period=151.5152}
  n22 = ScalarParameter() {defaultValue=-0.0066, group=Seared Scurf, parameterName=Speed X - Pattern 04}
  n23 = Multiply(A=Time() {period=45.4545}, B=ScalarParameter() {defaultValue=-0.2, group=Seared Scurf, parameterName=Speed Y - Pattern 04}) {constB=-0.022}
  n24 = ScalarParameter() {defaultValue=0.5, group=Seared Scurf, parameterName=Lava Overall Speed}
  n25 = AppendVector(
   A=Add(A=ComponentMask(None=@n20) {r=true}, B=Multiply(A=@n21, B=@n22) {constB=-0.0066})
   B=Add(A=ComponentMask(None=@n20) {g=true}, B=Multiply(A=@n23, B=@n24))
  )
  n26 = Multiply(
   A=TextureSample(UVs=AppendVector(A=@n16, B=@n19)) {samplerType=SAMPLERTYPE_LinearGrayscale, texture=&T_LavaLamp_Background}.R
   B=TextureSample(UVs=Multiply(A=@n25) {constB=-1}) {samplerType=SAMPLERTYPE_LinearGrayscale, texture=&T_LavaLamp_Background}.R
  )
  n27 = Multiply(A=@n14, B=@n26)
  n28 = Saturate(None=Multiply(A=StaticSwitchParameter(True=@n27, False=@n14) {defaultValue=true, group=Seared Scurf, parameterName=Combined/Seperate Layers}) {constB=2.5})
  n29 = Multiply(A=@n28, B=@n28)
  n30 = MaterialFunctionCall!(
   Fresnel=Multiply(A=@n29, B=@n29)
   Height=Multiply(A=OneMinus(None=@n28), B=ScalarParameter() {defaultValue=1, group=Seared Scurf, parameterName=Background Normal Intensity, sortPriority=6}) {constB=20}
  ) {materialFunction=&/MF_LavaNoise_Colors}
  n31 = Saturate(None=Multiply(A=@n26) {constB=2.5})
  n32 = Multiply(A=@n31, B=@n31)
  n33 = Multiply(A=@n32, B=@n32)
  n34 = MaterialFunctionCall(
   Fresnel=@n33
   Height=Multiply(A=OneMinus(None=@n31), B=ScalarParameter() {defaultValue=1, group=Seared Scurf, parameterName=Background Normal Intensity, sortPriority=6}) {constB=20}
  ) {materialFunction=&MF_LavaNoise_Colors}
  n35 = Saturate(None=Multiply(A=OneMinus(None=@n33)) {constB=7})
  FunctionOutput(None=StaticSwitchParameter(True=@n30.Bubble Colors, False=LinearInterpolate(A=@n30.Bubble Colors, B=@n34.Bubble Colors, Alpha=@n35)) {defaultValue=true, group=Seared Scurf, parameterName=Combined/Seperate Layers})
  NamedRerouteDeclaration(None=FunctionInput(Preview=TextureCoordinate()) {bUsePreviewValueAsDefault=true, inputName=UV, inputType=FunctionInput_Vector2}) {name=LavaUV, nodeColor=<rgba 1 0 0.5391 1>}
  StaticSwitchParameter(True=TextureCoordinate() {coordinateIndex=1}, False=TextureCoordinate()) {defaultValue=true, group=Seared Scurf, parameterName=Use UV2?}
 ]}
