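# Material function that outputs an animated emissive color for a potion, with glowing bubbles over a base color.
# A panning torus flowmap distorts two bubble-texture layers that scroll at separate tilings and opposite horizontal drift.
# Their maximum is multiplied by its inverted fourth power, favoring soft bubble edges over bright centers.
# The boosted green-channel mask tints the emissive color on front faces only, then adds the base color.
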MaterialFunction /Game/Global/Materials/MaterialGym/Materials/MaterialElements/UVs/AnimatedNoise/MF_PotionNoise
graph {nodes [
  n1 = Multiply(A=TextureCoordinate(), B=ScalarParameter() {defaultValue=3, group=Potion, parameterName=Potion-BubblesTile1})
  n2 = Multiply(A=Time(), B=ScalarParameter() {defaultValue=0.633, group=Potion, parameterName=Potion-BubblesSpeed})
  n3 = ScalarParameter() {defaultValue=0.05, group=Potion, parameterName=Potion-FlowmapStrength1}
  n4 = TextureCoordinate()
  n5 = ScalarParameter() {defaultValue=4, parameterName=Potion-FlowmapScale}
  n6 = Time()
  n7 = ScalarParameter() {defaultValue=0.2, group=Potion, parameterName=Potion-FlowmapSpeed}
  n8 = TextureSample(UVs=Panner(Coordinate=Multiply(A=@n4, B=@n5), Time=Multiply(A=@n6, B=@n7)) {speedX=1}) {samplerType=SAMPLERTYPE_LinearColor, texture=/Game/Global/Textures/Default/Flowmaps/T_TorusFlowmap}
  n9 = ComponentMask(None=Multiply(A=Subtract(A=@n8.RGB) {constB=0.5}, B=VectorParameter() {group=Potion, parameterName=Potion-FlowmapAmt, defaultValue=<rgba 1 1 0 1>}.RGB)) {g=true, r=true}
  n10 = TextureSample(UVs=Add(A=Panner(Coordinate=@n1, Time=@n2) {speedX=-0.1, speedY=1}, B=Multiply(A=@n3, B=@n9))) {samplerSource=SSM_Wrap_WorldGroupSettings, samplerType=SAMPLERTYPE_Masks, texture=/Game/Global/Textures/Default/PointSplatter/Bubbles_M}
  n11 = Multiply(A=TextureCoordinate(), B=ScalarParameter() {defaultValue=2, group=Potion, parameterName=Potion-BubblesTile2})
  n12 = ScalarParameter() {defaultValue=0.025, group=Potion, parameterName=Potion-FlowmapStrength2}
  n13 = TextureSample(UVs=Add(A=Panner(Coordinate=@n11, Time=@n2) {speedX=0.1, speedY=1}, B=Multiply(A=@n9, B=@n12))) {samplerSource=SSM_Wrap_WorldGroupSettings, samplerType=SAMPLERTYPE_Masks, texture=&Bubbles_M}
  n14 = Max(A=@n10.RGB, B=@n13.RGB)
  n15 = Power(Base=OneMinus(None=@n14)) {constExponent=4}
  n16 = Multiply(
   A=ComponentMask(None=Multiply(A=@n15, B=@n14)) {g=true}
   B=ScalarParameter() {defaultValue=7, group=Potion, parameterName=Potion-BubbleMaskAmount}
  )
  n17 = TwoSidedSign()
  n18 = Multiply(
   A=Multiply(A=Saturate(None=@n16), B=Saturate(None=@n17))
   B=VectorParameter() {group=Potion, parameterName=Potion-EmissiveColorA, defaultValue=<rgba 0 4.5567 12 1>}.RGB
  )
  n19 = VectorParameter() {group=Potion, parameterName=Potion-BaseColor, defaultValue=<rgba 0 0.1899 0.5 1>}
  FunctionOutput(None=NamedRerouteDeclaration(None=Add(A=@n18, B=@n19.RGB)) {name=EmissiveColor, nodeColor=<rgba 0.1484 1 0 1>})
 ]}
