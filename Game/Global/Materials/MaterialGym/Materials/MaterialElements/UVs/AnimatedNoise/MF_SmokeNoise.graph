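# Material function that outputs a tinted, panning smoke noise color, grouped as a SmokeBeneath layer with parallax depth.
# Input UVs are tiled, offset and panned, then sample the blue channel of a smoke map, with static switches for Bump Offset.
# That mask lerps between minimum and maximum brightness, then is multiplied by an emissive color, an invertible Fresnel term and a final noise tint.
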MaterialFunction /Game/Global/Materials/MaterialGym/Materials/MaterialElements/UVs/AnimatedNoise/MF_SmokeNoise
graph {nodes [
  n1 = ScalarParameter() {defaultValue=0.1, group=SmokeBeneath, parameterName=Smoke_BrightnessMIn}
  n2 = ScalarParameter() {defaultValue=2, group=SmokeBeneath, parameterName=Smoke_BrightnessMax}
  n3 = ScalarParameter() {group=SmokeBeneath, parameterName=Smoke_OffsetIntensity}
  n4 = AppendVector(
   A=ScalarParameter() {group=SmokeBeneath, parameterName=Smoke_UVOffsetU}
   B=ScalarParameter() {group=SmokeBeneath, parameterName=Smoke_UVOffsetV}
  )
  n5 = NamedRerouteDeclaration(None=FunctionInput(Preview=TextureCoordinate()) {bUsePreviewValueAsDefault=true, inputType=FunctionInput_Vector2}) {name=CachedBlendUV, nodeColor=<rgba 0.3359 0 1 1>}
  n6 = ScalarParameter() {defaultValue=1, group=SmokeBeneath, parameterName=Smoke_Tilling}
  n7 = Panner(
   Coordinate=Add(A=Multiply(A=@n3, B=@n4), B=Multiply(A=@n5, B=@n6))
   Speed=AppendVector(
    A=ScalarParameter() {group=SmokeBeneath, parameterName=Smoke_GradientSpeed_X}
    B=ScalarParameter() {defaultValue=0.05, group=SmokeBeneath, parameterName=Smoke_GradientSpeed_Y}
   )
  )
  n8 = ScalarParameter() {defaultValue=-3, group=SmokeBeneath, parameterName=Smoke_Bump Offset Height01}
  n9 = TextureSampleParameter2D(UVs=StaticSwitchParameter(True=BumpOffset(Coordinate=@n7, Height=@n8), False=@n7) {defaultValue=true, group=SmokeBeneath, parameterName=use Bump Offset on Smoke}) {
   group=SmokeBeneath
   parameterName=SmokeMap
   samplerSource=SSM_Wrap_WorldGroupSettings
   samplerType=SAMPLERTYPE_LinearColor
   texture=/Game/Accessories/FORT_Backpacks/Backpack_F_MED_TalonHime/Textures/T_Smokemap
  }
  n10 = TextureSampleParameter2D(UVs=@n7) {
   group=SmokeBeneath
   parameterName=SmokeMap
   samplerSource=SSM_Wrap_WorldGroupSettings
   samplerType=SAMPLERTYPE_LinearColor
   texture=&T_Smokemap
  }
  n11 = StaticSwitchParameter(True=@n9.B, False=@n10.B) {defaultValue=true, group=SmokeBeneath, parameterName=UseBumpOffsetMask}
  n12 = Saturate(None=Fresnel(ExponentIn=ScalarParameter() {defaultValue=4, group=SmokeBeneath, parameterName=Smoke_Emissive_Fresnel Exponent}) {baseReflectFraction=0})
  n13 = Multiply(
   A=Multiply(
    A=LinearInterpolate(A=@n1, B=@n2, Alpha=@n11)
    B=VectorParameter() {group=SmokeBeneath, parameterName=Smoke_Emissive, defaultValue=<rgba 1 1 1 1>}.RGB
   )
   B=LinearInterpolate(A=OneMinus(None=@n12), B=@n12, Alpha=ScalarParameter() {group=SmokeBeneath, parameterName=Smoke_Emissive_Invert Fresnel})
  )
  n14 = Constant() {r=1}
  FunctionOutput(None=Multiply(A=VectorParameter() {group=SmokeBeneath, parameterName=SmokeNoiseColor, defaultValue=<rgba 0.6 0 0.0126 1>}.RGB, B=Multiply(A=@n13, B=@n14)))
 ]}
