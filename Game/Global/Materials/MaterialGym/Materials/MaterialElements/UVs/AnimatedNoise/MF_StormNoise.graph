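# Material function that builds animated, lit storm noise, outputting a color and a Noise mask.
# A panning distortion texture offsets scrolling UVs for two grayscale noises and a normal map, sampled with explicit derivatives.
# The normal feeds a fresnel term and a sun highlight from the FortniteMaterialParameters sunlight vector, whose X also replaces Y.
# Masks blend dark to light color, add a tinted layer and boost brightness at peaks.
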MaterialFunction /Game/Global/Materials/MaterialGym/Materials/MaterialElements/UVs/AnimatedNoise/MF_StormNoise
graph {nodes [
  FunctionOutput(None=NamedRerouteUsage())
  n1 = NamedRerouteUsage()
  n2 = NamedRerouteUsage()
  n3 = Add(A=Saturate(None=Add(A=@n1, B=@n2)), B=NamedRerouteUsage())
  FunctionOutput(None=NamedRerouteDeclaration(None=Saturate(None=@n3)) {name=Noise_OUT, nodeColor=<rgba 1 0.5156 0 1>}) {outputName=Noise, sortPriority=1}
  NamedRerouteDeclaration(None=FunctionInput(Preview=TextureCoordinate()) {bUsePreviewValueAsDefault=true, inputName=UV, inputType=FunctionInput_Vector2}) {name=UVInput, nodeColor=<rgba 0.0781 0 1 1>}
  n4 = NamedRerouteUsage()
  n5 = FunctionInput(Preview=ComponentMask(None=Constant4Vector() {constant=<rgba 4 3 2 1>}.RGB) {g=true, r=true}) {bUsePreviewValueAsDefault=true, inputName=DistortionTiling, inputType=FunctionInput_Vector2, sortPriority=6}
  n6 = ComponentMask(None=Constant4Vector() {constant=<rgba -0.1 0.1 1 1>}.RGB) {g=true, r=true}
  n7 = TextureSample!(
   UVs=Panner(
    Coordinate=Multiply(A=@n4, B=@n5)
    Speed=FunctionInput(Preview=@n6) {bUsePreviewValueAsDefault=true, inputName=DistortionSpeed, inputType=FunctionInput_Vector2, sortPriority=6}
   )
   Tex=FunctionInput(Preview=TextureObject() {samplerType=SAMPLERTYPE_LinearColor, texture=/Game/Global/Textures/Default/Blanks/T_BlackLinearColor}) {bUsePreviewValueAsDefault=true, inputName=DistortionTexture, inputType=FunctionInput_Texture2D, sortPriority=1}
   "DDX(UVs)"=NamedRerouteUsage()
   "DDY(UVs)"=NamedRerouteUsage()
   Apply View MipBias
  ) {mipValueMode=TMVM_Derivative, samplerType=SAMPLERTYPE_LinearColor}
  NamedRerouteDeclaration(None=ComponentMask(None=ConstantBiasScale(None=@n7.RGB) {bias=-0.5, scale=2}) {g=true, r=true}) {name=UVDistortion, nodeColor=<rgba 1 0 0.75 1>}
  n8 = Constant4Vector() {constant=<rgba 1 1 1 1>}
  n9 = Multiply(
   A=NamedRerouteUsage()
   B=FunctionInput(Preview=ComponentMask(None=@n8.RGB) {g=true, r=true}) {bUsePreviewValueAsDefault=true, inputName=NoiseTiling, inputType=FunctionInput_Vector2, sortPriority=5}
  )
  n10 = Multiply(A=Constant() {r=0.02}, B=NamedRerouteUsage())
  n11 = Time()
  n12 = FunctionInput(Preview=ComponentMask(None=Constant4Vector() {constant=<rgba 0.025 -0.05 1 1>}.RGB) {g=true, r=true}) {bUsePreviewValueAsDefault=true, inputName=NoiseSpeed, inputType=FunctionInput_Vector2, sortPriority=5}
  NamedRerouteDeclaration(None=Add(A=Add(A=@n9, B=@n10), B=Multiply(A=@n11, B=@n12))) {name=UVs, nodeColor=<rgba 0.8516 1 0 1>}
  n13 = CollectionParameter() {
   collection=&FortniteMaterialParameters
   parameterName=SunlightDirectionalVector
  }
  n14 = MaterialFunctionCall(Float3=ComponentMask(None=NamedRerouteDeclaration(None=@n13) {name=LightVecWS, nodeColor=<rgba 0 1 0.9297 1>}) {b=true, g=true, r=true}) {materialFunction=&BreakOutFloat3Components}
  n15 = Max(A=@n14.B) {constB=0.2}
  n16 = DotProduct(
   A=Normalize(VectorInput=MaterialFunctionCall(X=@n14.R, Y=@n14.R, Z=@n15) {materialFunction=&MakeFloat3}.Result)
   B=Transform(None=NamedRerouteUsage())
  )
  NamedRerouteDeclaration(None=Power(Base=Saturate(None=@n16), Exp=Constant() {r=5})) {name=LightWave, nodeColor=<rgba 0 1 0.5078 1>}
  n17 = TextureSample(
   UVs=NamedRerouteUsage()
   Tex=FunctionInput(Preview=TextureObject() {samplerType=SAMPLERTYPE_LinearGrayscale, texture=/Game/Global/Textures/Default/Blanks/T_BlackLinearGrayscale}) {bUsePreviewValueAsDefault=true, inputName=NoiseTexture02, inputType=FunctionInput_Texture2D, sortPriority=1}
   "DDX(UVs)"=NamedRerouteUsage()
   "DDY(UVs)"=NamedRerouteUsage()
  ) {mipValueMode=TMVM_Derivative, samplerType=SAMPLERTYPE_LinearGrayscale, texture=&FlatNormalBlue}
  NamedRerouteDeclaration(None=@n17.RGB) {name=NoiseSampleTint, nodeColor=<rgba 1 0 0.3281 1>}
  n18 = TextureObject() {samplerType=SAMPLERTYPE_Normal, texture=&FlatNormalBlue}
  NamedRerouteDeclaration(None=TextureSample(
    UVs=NamedRerouteUsage()
    Tex=FunctionInput(Preview=@n18) {bUsePreviewValueAsDefault=true, inputName=NoiseNormal, inputType=FunctionInput_Texture2D, sortPriority=1}
    "DDX(UVs)"=NamedRerouteUsage()
    "DDY(UVs)"=NamedRerouteUsage()
   ) {mipValueMode=TMVM_Derivative, samplerType=SAMPLERTYPE_Normal}.RGB) {name=NoiseNormalSample, nodeColor=<rgba 1 0 0.3516 1>}
  n19 = TextureSample(
   UVs=NamedRerouteUsage()
   Tex=FunctionInput(Preview=TextureObject() {samplerType=SAMPLERTYPE_LinearGrayscale, texture=&T_BlackLinearGrayscale}) {bUsePreviewValueAsDefault=true, inputName=NoiseTexture01, inputType=FunctionInput_Texture2D, sortPriority=1}
   "DDX(UVs)"=NamedRerouteUsage()
   "DDY(UVs)"=NamedRerouteUsage()
  ) {mipValueMode=TMVM_Derivative, samplerType=SAMPLERTYPE_LinearGrayscale, texture=&FlatNormalBlue}
  NamedRerouteDeclaration(None=NamedRerouteDeclaration(None=@n19.RGB) {name=NoiseSample, nodeColor=<rgba 0 0.3437 1 1>}) {name=FresnelMask, nodeColor=<rgba 1 0.7031 0 1>}
  n20 = NamedRerouteUsage()
  n21 = OneMinus(None=DotProduct(A=CameraVectorWS(), B=Transform(None=@n20)))
  NamedRerouteDeclaration(None=Power(Base=ConstantBiasScale(None=@n21))) {name=NoiseNormalFresnel, nodeColor=<rgba 0 1 0.0859 1>}
  n22 = FunctionInput(Preview=Constant3Vector() {constant=<rgba 0.2582 0.5647 0.6939 1>}) {bUsePreviewValueAsDefault=true, inputName=NoiseColorLight, sortPriority=3}
  n23 = Multiply(
   A=FunctionInput(Preview=Constant() {r=0.2}) {bUsePreviewValueAsDefault=true, inputName=NoiseColorDarkOffset, inputType=FunctionInput_Scalar, sortPriority=3}
   B=@n22
  )
  n24 = Saturate(None=Add(A=NamedRerouteUsage(), B=NamedRerouteUsage()))
  n25 = Power(Base=@n24, Exp=Constant() {r=1.3}) {constExponent=1.5}
  n26 = Multiply(A=@n24) {constB=0.9}
  n27 = Constant() {r=6}
  n28 = Multiply(
   A=Multiply(A=LinearInterpolate(A=@n23, B=@n22, Alpha=@n25), B=NamedRerouteUsage())
   B=LinearInterpolate(A=Constant() {r=1}, B=Constant() {r=3}, Alpha=Power(Base=@n26, Exp=@n27) {constExponent=4})
  )
  n29 = NamedRerouteUsage()
  n30 = Constant4Vector() {constant=<rgba 1 0.1389 0.9332 1>}
  n31 = Constant() {r=6}
  n32 = Multiply(
   A=Multiply(A=@n29, B=FunctionInput(Preview=@n30.RGB) {bUsePreviewValueAsDefault=true, inputName=NoiseTint, sortPriority=3})
   B=LinearInterpolate(A=Constant() {r=2}, B=Constant() {r=5}, Alpha=Power(Base=@n29, Exp=@n31))
  )
  NamedRerouteDeclaration(None=Multiply(A=Add(A=@n28, B=@n32), B=Constant() {r=1})) {name=Output, nodeColor=<rgba 0.0312 0 1 1>}
  n33 = TextureCoordinate()
  NamedRerouteDeclaration(None=DDX(Value=@n33)) {name=DDX, nodeColor=<rgba 1 0.2109 0 1>}
  NamedRerouteDeclaration(None=DDY(Value=@n33)) {name=DDY, nodeColor=<rgba 0 0.1094 1 1>}
 ]}
