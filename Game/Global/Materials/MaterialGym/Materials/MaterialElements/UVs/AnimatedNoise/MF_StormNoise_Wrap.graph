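# Material function that wraps MF_StormNoise, exposing its inputs as named material parameters in the MF_StormNoise group for animated storm noise.
# Texture parameters supply a normal, two grayscale noises and a distortion map, all with blank or flat defaults, alongside color, tint, tiling and speed parameters.
# An optional InputUV input falls back to the default texture coordinates, and the inner function's result and NoiseMask pass through as outputs.
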MaterialFunction /Game/Global/Materials/MaterialGym/Materials/MaterialElements/UVs/AnimatedNoise/MF_StormNoise_Wrap
graph {nodes [
  n1 = TextureObjectParameter() {
   group=MF_StormNoise
   parameterName=StormNoiseNormal
   samplerType=SAMPLERTYPE_Normal
   sortPriority=3
   texture=&FlatNormalBlue
  }
  n2 = TextureObjectParameter() {
   group=MF_StormNoise
   parameterName=StormNoiseTexture01
   samplerType=SAMPLERTYPE_LinearGrayscale
   sortPriority=3
   texture=/Game/Global/Textures/Default/Blanks/T_BlackLinearGrayscale
  }
  n3 = TextureObjectParameter() {
   group=MF_StormNoise
   parameterName=StormNoiseTexture02
   samplerType=SAMPLERTYPE_LinearGrayscale
   sortPriority=3
   texture=&T_BlackLinearGrayscale
  }
  n4 = TextureObjectParameter() {
   group=MF_StormNoise
   parameterName=StormDistortionTexture
   samplerType=SAMPLERTYPE_LinearColor
   sortPriority=3
   texture=/Game/Global/Textures/Default/Blanks/T_BlackLinearColor
  }
  n5 = VectorParameter() {group=MF_StormNoise, parameterName=StormNoiseColorLight, sortPriority=1, defaultValue=<rgba 0.2646 0.5736 0.7011 1>}
  n6 = VectorParameter() {group=MF_StormNoise, parameterName=StormNoiseTint, sortPriority=1, defaultValue=<rgba 1 0.1389 0.9332 1>}
  n7 = VectorParameter() {group=MF_StormNoise, parameterName=StormNoiseTiling, channelNames {a="", b="", g=Tile Y, r=Tile X}, defaultValue=<rgba 1 1 1 1>}
  n8 = VectorParameter() {group=MF_StormNoise, parameterName=StormDistortionTiling, channelNames {a="", b="", g=Tile Y, r=Tile X}, defaultValue=<rgba 4 3 2 1>}
  n9 = VectorParameter() {
   group=MF_StormNoise
   parameterName=StormDistortionSpeed
   sortPriority=4
   channelNames {a="", b="", g=Speed Y, r=Speed X}
   defaultValue=<rgba -0.1 0.1 1 1>
  }
  n10 = MaterialFunctionCall!(
   UV=FunctionInput(Preview=TextureCoordinate()) {bUsePreviewValueAsDefault=true, inputName=InputUV, inputType=FunctionInput_Vector2}
   NoiseNormal=@n1
   NoiseTexture01=@n2
   NoiseTexture02=@n3
   DistortionTexture=@n4
   NoiseColorLight=@n5.RGB
   NoiseColorDarkOffset=ScalarParameter() {defaultValue=0.2, group=MF_StormNoise, parameterName=StormNoiseColorDarkOffset, sortPriority=1}
   NoiseTint=@n6.RGB
   NoiseTiling=@n7
   NoiseSpeed=VectorParameter() {group=MF_StormNoise, parameterName=StormNoiseSpeed, defaultValue=<rgba 0.025 -0.05 1 1>}.RGB
   DistortionTiling=@n8
   DistortionSpeed=@n9
  ) {materialFunction=&/MF_StormNoise}
  FunctionOutput(None=@n10.Result)
  FunctionOutput(None=@n10.Noise) {outputName=NoiseMask, sortPriority=1}
 ]}
