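# Parent Default Lit material in MaterialGym's animated-noise UV elements for a storm noise effect, flagged for skeletal meshes and mesh deformers.
# The graph is a thin shell holding a Constant, MF_StormNoise_Wrap and MF_StormNoise, with no wiring between them.
# MF_StormNoise exposes inputs for UV, normal, two noise textures, a distortion texture, colors and tint, plus tiling and speed for noise and distortion.
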Material /Game/Global/Materials/MaterialGym/Materials/MaterialElements/UVs/AnimatedNoise/M_StormNoise_Parent
graph {
 properties {bAutomaticallySetUsageInEditor=false, bUsedWithMeshDeformer=true, bUsedWithSkeletalMesh=true, usedShadingModels=Default Lit}
 nodes [
  Constant()
  MaterialFunctionCall!(InputUV) {materialFunction=&/MF_StormNoise_Wrap}
  MaterialFunctionCall!(
   UV
   NoiseNormal
   NoiseTexture01
   NoiseTexture02
   DistortionTexture
   NoiseColorLight
   NoiseColorDarkOffset
   NoiseTint
   NoiseTiling
   NoiseSpeed
   DistortionTiling
   DistortionSpeed
  ) {materialFunction=&/MF_StormNoise}
 ]
}
