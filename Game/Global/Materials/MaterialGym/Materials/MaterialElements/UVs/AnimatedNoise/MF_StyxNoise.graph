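# Material function that builds the animated Styx noise emissive color, with default textures from a player body cosmetic.
# Scaled UVs pass through BumpOffset for parallax, and a panning normal map supplies UV distortion for the noise.
# Styx mask and detail textures yield a contrast-shaped main pattern, detail noise, macro mask and edge mask.
# Tinted main, masked detail glow and two-color macro layers are summed and scaled by an overall intensity.
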MaterialFunction /Game/Global/Materials/MaterialGym/Materials/MaterialElements/UVs/AnimatedNoise/MF_StyxNoise
graph {nodes [
  n1 = NamedRerouteUsage()
  n2 = VectorParameter() {group=Styx, parameterName=StyxEmissive_DetailGlow, defaultValue=<rgba 0.2 8 2.6 1>}
  n3 = Max(A=Desaturation(None=@n1), B=ScalarParameter() {defaultValue=0.02, group=Styx, parameterName=Styx_Detail_Noise_MinValue}) {constB=0.1}
  n4 = NamedRerouteUsage()
  n5 = NamedRerouteDeclaration(None=Multiply(A=Add(A=@n3, B=@n4), B=NamedRerouteUsage())) {name=DetailMask, nodeColor=<rgba 1 0.6797 0 1>}
  n6 = Add(A=@n1, B=Multiply(A=Multiply(A=@n2.RGB, B=@n5), B=ScalarParameter() {defaultValue=1, group=Styx, parameterName=DetailNoiseIntensity}))
  n7 = Multiply(
   A=ScalarParameter() {defaultValue=1, group=Styx, parameterName=MacrNoiseIntensity}
   B=LinearInterpolate(
    A=VectorParameter() {group=Styx, parameterName=StyxEmissive_MacroNoise_A, defaultValue=<rgba 0.003 0.039 0.079 1>}.RGB
    B=VectorParameter() {group=Styx, parameterName=StyxEmissive_MacroNoise_B, defaultValue=<rgba 0.172 0.81 0.08 1>}.RGB
    Alpha=NamedRerouteUsage()
   )
  )
  FunctionOutput(None=Multiply(A=ScalarParameter() {defaultValue=0.7, group=Styx, parameterName=OverallEmissiveIntensity}, B=Add(A=@n6, B=@n7))) {outputName=OutPutEmissiveColor}
  n8 = FunctionInput(Preview=TextureCoordinate()) {bUsePreviewValueAsDefault=true, inputName=UV, inputType=FunctionInput_Vector2}
  n9 = MaterialFunctionCall(Float2=Multiply(A=@n8, B=ScalarParameter() {defaultValue=2.5, group=Styx, parameterName=StyxMainNoiseTilling})) {materialFunction=&BreakOutFloat2Components}
  n10 = MaterialFunctionCall(X=@n9.R, Y=@n9.R) {materialFunction=&MakeFloat2}
  NamedRerouteDeclaration(None=BumpOffset(
    Coordinate=NamedRerouteDeclaration(None=@n10.Result) {name=ModUV, nodeColor=<rgba 0 1 0.9297 1>}
    Height=ScalarParameter() {group=Styx, parameterName=StyxUVOffset01}
   )) {name=UV_withParallax_A, nodeColor=<rgba 1 0.6562 0 1>}
  n11 = NamedRerouteUsage()
  n12 = AppendVector(
   A=ScalarParameter() {group=Styx, parameterName=StyxNoise_Speed2X}
   B=ScalarParameter() {defaultValue=-0.15, group=Styx, parameterName=StyxNoise_Speed2Y}
  )
  n13 = TextureSample(UVs=Add(A=Panner(Coordinate=@n11, Speed=@n12), B=NamedRerouteUsage()), Tex=NamedRerouteUsage()) {samplerSource=SSM_Wrap_WorldGroupSettings}
  n14 = Multiply(
   A=VectorParameter() {group=Styx, parameterName=StyxEmissive2, defaultValue=<rgba 0.0323 0.2875 0.0757 1>}.RGB
   B=Power(Base=@n13.R, Exp=ScalarParameter() {defaultValue=3, group=Styx, parameterName=StyxNoise_Contrast2})
  )
  n15 = Multiply(A=VectorParameter() {group=Styx, parameterName=StyxEmissive, defaultValue=<rgba 0.0311 0.8385 0.0503 1>}.RGB, B=@n13.R)
  n16 = Saturate(None=Power(Base=@n13.R, Exp=ScalarParameter() {defaultValue=1.37, group=Styx, parameterName=StyxNoise_Contrast}))
  NamedRerouteDeclaration(None=Multiply(A=LinearInterpolate(A=@n14, B=@n15, Alpha=@n16), B=ScalarParameter() {defaultValue=1, group=Styx, parameterName=MainNoiseIntensity})) {name=MainNoisePattern, nodeColor=<rgba 0 1 0.3203 1>}
  n17 = ScalarParameter() {defaultValue=-0.1, group=Styx, parameterName=Styx_Detail_Noise_Low}
  n18 = ScalarParameter() {defaultValue=1, group=Styx, parameterName=Styx_Detail_Noise_High}
  n19 = NamedRerouteUsage()
  n20 = NamedRerouteUsage()
  n21 = ScalarParameter() {group=Styx, parameterName=Styx_Detail_Noise_SpeedX}
  n22 = ScalarParameter() {defaultValue=-0.1, group=Styx, parameterName=Styx_Detail_Noise_SpeedY}
  n23 = TextureSampleParameter2D(UVs=Panner(Coordinate=Add(A=@n19, B=@n20), Speed=AppendVector(A=@n21, B=@n22))) {
   group=Styx
   parameterName=StyxNoiseDetail
   samplerType=SAMPLERTYPE_Masks
   texture=/Game/Characters/Player/Male/Medium/Bodies/M_MED_DarkStance_Pater/Textures/T_StyxCosmetic_DetailNoise_M
  }
  NamedRerouteDeclaration(None=Saturate(None=LinearInterpolate(A=@n17, B=@n18, Alpha=@n23.R))) {name=DetailNoise, nodeColor=<rgba 1 0.0938 0 1>}
  NamedRerouteDeclaration(None=Saturate(None=@n23.G)) {name=MarcoNoiseMask, nodeColor=<rgba 0 1 0.8359 1>}
  n24 = NamedRerouteUsage()
  n25 = Multiply(A=NamedRerouteUsage(), B=ScalarParameter() {defaultValue=0.1, group=Styx, parameterName=Styx_NoiseEdge_Distortion})
  n26 = LinearInterpolate(
   A=ScalarParameter() {group=Styx, parameterName=EdgeGlow_Noise_Contrast_Low}
   B=ScalarParameter() {defaultValue=1, group=Styx, parameterName=EdgeGlow_Noise_Contrast_HIGH}
   Alpha=TextureSample(UVs=Add(A=@n24, B=@n25), Tex=NamedRerouteUsage()).G
  )
  NamedRerouteDeclaration(None=NamedRerouteDeclaration(None=Saturate(None=@n26)) {name=NoisePerlin, nodeColor=<rgba 0.2188 0 1 1>}) {name=EdgeMask, nodeColor=<rgba 0.2422 0 1 1>}
  n27 = MaterialFunctionCall(Float2=Multiply(A=@n8, B=ScalarParameter() {defaultValue=4, group=Styx, parameterName=StyxMainNoiseTilling_N})) {materialFunction=&BreakOutFloat2Components}
  n28 = MaterialFunctionCall(X=@n27.R, Y=@n27.R) {materialFunction=&MakeFloat2}
  NamedRerouteDeclaration(None=BumpOffset(
    Coordinate=NamedRerouteDeclaration(None=@n28.Result) {name=ModUV_N, nodeColor=<rgba 0 1 0.9297 1>}
    Height=ScalarParameter() {group=Styx, parameterName=StyxUVOffset01}
   )) {name=UV_withParallax_N, nodeColor=<rgba 1 0.6562 0 1>}
  n29 = ScalarParameter() {group=Styx, parameterName=StyxNoise_SpeedX}
  n30 = ScalarParameter() {defaultValue=-0.15, group=Styx, parameterName=StyxNoise_SpeedY}
  n31 = TextureObjectParameter() {
   group=Styx
   parameterName=StyxDistortNormal
   samplerType=SAMPLERTYPE_Normal
   texture=/Game/Characters/Player/Male/Medium/Bodies/M_MED_DarkStance_Pater/Textures/T_StyxCosmetic_N
  }
  n32 = TextureSample(UVs=Panner(Coordinate=NamedRerouteUsage(), Speed=AppendVector(A=@n29, B=@n30)), Tex=@n31) {
   samplerSource=SSM_Wrap_WorldGroupSettings
   samplerType=SAMPLERTYPE_Normal
   texture=&T_StyxCosmetic_N
  }
  NamedRerouteDeclaration(None=ComponentMask(None=@n32.RGB) {g=true, r=true}) {name=Distortion_NormalNoise_A, nodeColor=<rgba 0.3828 0 1 1>}
  n33 = NamedRerouteUsage()
  n34 = ScalarParameter() {defaultValue=0.03, group=Styx, parameterName=Styx_Distort_UV}
  NamedRerouteDeclaration(None=Add(A=Multiply(A=@n33, B=@n34), B=Constant())) {name=Distortion_Final_A, nodeColor=<rgba 0.2422 0 1 1>}
  n35 = TextureObjectParameter() {
   group=Styx
   parameterName=Styx Mask
   samplerType=SAMPLERTYPE_Masks
   texture=/Game/Characters/Player/Male/Medium/Bodies/M_MED_DarkStance_Pater/Textures/T_StyxCosmetic_Mask
  }
  NamedRerouteDeclaration(None=@n35) {name=Styx Mask IN, nodeColor=<rgba 1 0.8906 0 1>}
 ]}
