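# M_Styx is a Default Lit material in the MaterialGym animated-noise UV elements folder, built around the MF_StyxNoise material function.
# The graph holds only that function call, which exposes a UV input, and a Constant, with no wiring shown between them or to material outputs.
# It is flagged for skeletal mesh and mesh deformer use, with automatic usage setting in the editor turned off.
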Material /Game/Global/Materials/MaterialGym/Materials/MaterialElements/UVs/AnimatedNoise/M_Styx
graph {
 properties {bAutomaticallySetUsageInEditor=false, bUsedWithMeshDeformer=true, bUsedWithSkeletalMesh=true, usedShadingModels=Default Lit}
 nodes [
  Constant()
  MaterialFunctionCall!(UV) {materialFunction=&/MF_StyxNoise}
 ]
}
