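# Material function that outputs an animated, HDR-tinted lightning noise pattern, exposed through Lightning group parameters.
# A panning noise feeds a SphereMask around 0.5, whose sine-pulsed radius turns drifting noise bands into flickering masks.
# These mask a power-sharpened lightning texture, whose panning UVs are distorted by a second scrolling noise.
# The UVInput only feeds an unused reroute, so all panners use the default texture coordinates.
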MaterialFunction /Game/Global/Materials/MaterialGym/Materials/MaterialElements/UVs/AnimatedNoise/MF_LightningNoise
graph {nodes [
  n1 = ScalarParameter() {defaultValue=0.02, group=Lightning, parameterName=LightningXSpeed}
  n2 = ScalarParameter() {defaultValue=0.1, group=Lightning, parameterName=LightningYSpeed}
  n3 = Multiply(A=Panner(Coordinate=TextureCoordinate(), Speed=AppendVector(A=@n1, B=@n2)) {speedX=0.02, speedY=0.1}) {constB=2}
  n4 = Sine(None=Multiply(A=Time(), B=ScalarParameter() {defaultValue=1, group=Lightning, parameterName=LightningSpeed})) {period=0.1}
  n5 = ScalarParameter() {defaultValue=0.3, group=Lightning, parameterName=LightningRadius}
  n6 = Panner(Coordinate=TextureCoordinate()) {speedX=0.05, speedY=0.1}
  n7 = Constant() {r=0.5}
  n8 = Multiply(A=TextureSample(UVs=Multiply(A=@n6, B=@n7)) {samplerSource=SSM_Wrap_WorldGroupSettings, texture=/Game/Global/Textures/Default/Noise/T_Tiling_Noise_01}.R) {constB=0.2}
  n9 = Panner(Coordinate=TextureCoordinate()) {speedX=0.02, speedY=0.3}
  n10 = TextureSampleParameter2D(UVs=Multiply(A=Add(A=@n8, B=@n9), B=ScalarParameter() {defaultValue=1, group=Lightning, parameterName=LightningNoiseTiling})) {
   group=Lightning
   parameterName=Lightning Texture
   texture=/Game/Global/Materials/MaterialGym/Materials/MaterialElements/UVs/AnimatedNoise/Texture/LightningFlip_T2
  }
  n11 = ScalarParameter() {defaultValue=4, group=Lightning, parameterName=LightningPow}
  n12 = Multiply(
   A=SphereMask(
    A=TextureSample(UVs=@n3) {samplerSource=SSM_Wrap_WorldGroupSettings, texture=&T_Tiling_Noise_01}.R
    B=Constant() {r=0.5}
    Radius=Multiply(A=@n4, B=@n5) {constB=0.8}
    Hardness=ScalarParameter() {defaultValue=0.7, group=Lightning, parameterName=LightningHardness}
   )
   B=Max(B=Power(Base=@n10.R, Exp=@n11) {constExponent=3})
  )
  FunctionOutput(None=Multiply(
    A=Power(Base=@n12) {constExponent=1}
    B=VectorParameter() {group=Lightning, parameterName=Lightning Color, defaultValue=<rgba 25 8.242 24.1763 1>}.RGB
   )) {outputName=Lightning Out}
  NamedRerouteDeclaration(None=FunctionInput(Preview=TextureCoordinate()) {bUsePreviewValueAsDefault=true, inputName=UVInput, inputType=FunctionInput_Vector2}) {name=UV, nodeColor=<rgba 0.8984 0 1 1>}
 ]}
