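# M_lightningNoise is a Default Lit material in the MaterialGym's animated-noise UV elements, built around the MF_LightningNoise material function.
# The graph holds only a call to that function, which exposes a UV input, and a single constant, leaving the noise logic inside the function.
# The material is flagged as having pixel animation, and automatic usage setting in the editor is off.
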Material /Game/Global/Materials/MaterialGym/Materials/MaterialElements/UVs/AnimatedNoise/M_lightningNoise
graph {
 properties {bAutomaticallySetUsageInEditor=false, bHasPixelAnimation=true, usedShadingModels=Default Lit}
 nodes [
  Constant()
  MaterialFunctionCall!(UVInput) {materialFunction=&/MF_LightningNoise}
 ]
}
