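# M_ScratchPad is a Material Gym scratch-pad material for trying UV projection techniques, flagged for skeletal meshes and mesh deformers.
# Shared texture, tiling, offset, blend contrast and depth parameters, routed through named reroutes, supply material functions for pre-skinned triplanar, camera-aligned virtual plane and triplanar virtual box sampling.
# Nested static switches, UseVirtualBox and UseVirtualPlane, select between the techniques, with a UV color grid as the default texture.
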Material /Game/Global/Materials/MaterialGym/Materials/MaterialElements/UVs/M_ScratchPad
graph {
 properties {bAutomaticallySetUsageInEditor=false, bUsedWithMeshDeformer=true, bUsedWithSkeletalMesh=true, usedShadingModels=Default Lit}
 nodes [
  NamedRerouteDeclaration(None=AppendVector(
    A=ScalarParameter() {defaultValue=1, group=ScratchPad - UVs, parameterName=UVTilingX, sortPriority=1}
    B=ScalarParameter() {defaultValue=1, group=ScratchPad - UVs, parameterName=UVTilingY, sortPriority=2}
   )) {name=UVTiling, nodeColor=<rgba 1 0.4922 0 1>}
  NamedRerouteDeclaration(None=AppendVector(
    A=ScalarParameter() {group=ScratchPad - UVs, parameterName=UVOffsetX, sortPriority=3}
    B=ScalarParameter() {group=ScratchPad - UVs, parameterName=UVOffsetY, sortPriority=4}
   )) {name=UVOffset, nodeColor=<rgba 1 0.4922 0 1>}
  NamedRerouteDeclaration(None=TextureObjectParameter() {group=ScratchPad - UVs, parameterName=Texture, sortPriority=0, texture=/UVEditor/Textures/UVEditorColorGrid_Color}) {name=Texture, nodeColor=<rgba 1 0.4922 0 1>}
  NamedRerouteDeclaration(None=ScalarParameter() {defaultValue=1, group=ScratchPad - UVs, parameterName=BlendContrast, sortPriority=7}) {name=BlendContrast, nodeColor=<rgba 1 0 0.4687 1>}
  NamedRerouteDeclaration(None=MaterialFunctionCall!(InputTexture=NamedRerouteUsage(), UV Tiling=NamedRerouteUsage(), UV Offset=NamedRerouteUsage(), Blend Contrast=NamedRerouteUsage()) {materialFunction=/Game/Global/Materials/MaterialGym/Materials/MaterialElements/UVs/PreskinnedTriplanar/MF_PreskinnedTriplanar}.Sample XYZ) {name=Triplanar, nodeColor=<rgba 1 0 0.4219 1>}
  n1 = MaterialFunctionCall!(UV Tiling=NamedRerouteUsage(), UV Offset=NamedRerouteUsage()) {materialFunction=/Game/Global/Materials/MaterialGym/Materials/MaterialElements/UVs/CameraAlignedVirutalPlane/MF_CameraAlignedVirtualPlane}
  NamedRerouteDeclaration(None=TextureSample(UVs=@n1.OutputUV, Tex=NamedRerouteUsage()) {texture=&UVEditorColorGrid_Color}.RGB) {name=VirutalPlane, nodeColor=<rgba 1 0 0.4219 1>}
  NamedRerouteDeclaration(None=MaterialFunctionCall!(
    TVB_Texture=NamedRerouteUsage()
    TVB_UV_Tiling=NamedRerouteUsage()
    TVB_UV_Offset=NamedRerouteUsage()
    TVB_CenterPointOffset
    TVB_Depth=NamedRerouteUsage()
    TVB_BlendContrast=NamedRerouteUsage()
    TVB_UseLocalCameraVector
   ) {materialFunction=/Game/Global/Materials/MaterialGym/Materials/MaterialElements/UVs/PreskinnedVirtualBox/MF_TriplanarVirtualBox}.Result) {name=VirtualBox, nodeColor=<rgba 0 1 0.1328 1>}
  NamedRerouteDeclaration(None=ScalarParameter() {group=ScratchPad - UVs, parameterName=Depth}) {name=Depth, nodeColor=<rgba 1 0.2812 0 1>}
  StaticSwitchParameter(
   True=NamedRerouteUsage()
   False=StaticSwitchParameter(True=NamedRerouteUsage(), False=NamedRerouteUsage()) {group=ScratchPad - UVs, parameterName=UseVirtualPlane}
  ) {group=ScratchPad - UVs, parameterName=UseVirtualBox}
 ]
}
